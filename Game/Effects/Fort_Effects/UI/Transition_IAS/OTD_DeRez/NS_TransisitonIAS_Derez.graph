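# Niagara de-rez transition effect that breaks the User.TargetPawn skeletal mesh into drifting, tumbling mesh and sprite particles with blue tinting.
# Two burst emitters scatter particles in spheres around 14 sampled bone groups, and a third trails blue ribbons from the sparser burst.
# An age-animated height offset from the head drives per-particle alpha, which gates curl noise, point and line-attraction forces.
# StartEdge-indexed curves stage motion, while particles fade near the origin.
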NiagaraSystem /Game/Effects/Fort_Effects/UI/Transition_IAS/OTD_DeRez/NS_TransisitonIAS_Derez
dependencies {
 usedDataInterfaces [
  &NiagaraDataInterfaceArrayFloat
  &NiagaraDataInterfaceCurve
  &NiagaraDataInterfaceParticleRead
  &NiagaraDataInterfaceColorCurve
 ]
 usedDynamicInputs []
 usedModules [&SystemState, &SampleSkeletalMesh]
}
emitters {
 Empty {
  emitterUpdateScript [
   &EmitterState(Life Cycle Mode=::NewEnumerator1, Loop Behavior=::NewEnumerator1, Loop Duration=2, UseLoopDelay=false)
   &SpawnBurst_Instantaneous(Spawn Count=98, Use Spawn Probability=false, UseLoopCountLimit=false)
  ]
  particleSpawnScript [
   &FN_SetSpawnGroup(Inclusivity=true, Normalize Spawn Group=true, Number of Spawn Groups=14, Particles Per Group=7)
   &InitializeParticle(
    Lifetime=4
    Mesh Scale Mode=::NewEnumerator1
    Mesh Uniform Scale Max=2.25
    Mesh Uniform Scale Min=1.15
    Sprite Size Mode=::NewEnumerator3
    Uniform Sprite Size=13.0186
    UsePositionOffset=false
    Write Material Random=true
   )
   &self:CustomBoneSample(
    NormalizedSpawnGroup:NiagaraFloat=$Particles.Normalized Spawn Group
    Skeletal Mesh:NiagaraDataInterfaceSkeletalMesh=$User.TargetPawn
    SpawnGroup:NiagaraInt32=$Particles.SpawnGroup
   ) {module=CustomBoneSample}
   &ShapeLocation(
    MaskSpawnGroup=false
    Override Seed=false
    Shape Origin=$StackContext.SampledPosition
    Sphere Radius=&Multiply_Float(
     A=&SelectFloatFromArray(
      Direct Array Index=$Particles.SpawnGroup
      Float Selection Array=DataInterface() {propertyValues {FloatData [10,10,10,7,7,5,5,18,10,13,13,8,8,18]}}
     )
     B=2
    )
    UseOwnerScale=true
   )
  ]
  particleUpdateScript [
   &ParticleState()
   &self:AlphaDependOnHeight(
    DivideZ=80
    HeadPos:Vector3f=$System.SampleSkeletalMesh.SampledPosition
    ZOffset:NiagaraFloat=&Lerp_Float(
     A=250
     B=0
     Alpha=&ClampFloat(Float=&Subtract_Float(A=1, B=&Multiply_Float(A=$Emitter.Age, B=1)))
    )
   ) {module=AlphaDependOnHeight}
   &self:AlphaDependOnHeight(
    DivideZ=70
    ZOffset:NiagaraFloat=&Lerp_Float(
     A=2665.4104
     B=7.5492
     Alpha=&ClampFloat(Float=&Subtract_Float(A=1, B=&Multiply_Float(A=$Emitter.Age, B=1)))
    )
   ) {enabled=false, module=AlphaDependOnHeight001}
   &UpdateMeshOrientation(
    Rotation Rate=&Lerp_Float(
     A=0
     B=&RandomRangeFloat(Minimum=-3, Maximum=3)
     Alpha=&FloatFromCurve(FloatCurve=DataInterface() {propertyValues {Curve {keys [<at 0 0>,<at 0.1 1>]}}}, CurveIndex=$Particles.StartEdge)
    )
    Rotation Vector=&RandomRangeVector(Minimum=<-1 -1 -1>, Maximum=<1 1 1>)
   )
   &ScaleColor(Scale Alpha=$Particles.Alpha, ScaleA=true, ScaleRGB=true)
   &CurlNoiseForce(
    Noise Frequency=7
    Noise Strength=&FloatFromCurve(
     FloatCurve=DataInterface() {propertyValues {Curve {keys [
         {interpMode=RCIM_Cubic, tangentMode=RCTM_Break, time=0.1, value=1}
         {arriveTangent=0.1541, interpMode=RCIM_Cubic, tangentMode=RCTM_Break, time=1.8, value=0.3}
        ]}}}
     CurveIndex=$Particles.StartEdge
     Scale Curve=16000
    )
    Pan Noise=true
    Pan Noise Field=<0 0 0.1>
   )
   &PointForce(
    Force Strength=&FloatFromCurve(
     FloatCurve=DataInterface() {propertyValues {Curve {keys [{interpMode=RCIM_Cubic, leaveTangent=-1.6096, tangentMode=RCTM_Break, time=0, value=-0.5}, <at 0.5 -1>]}}}
     CurveIndex=$Particles.StartEdge
     Scale Curve=9000
    )
    Module.Use Falloff=false
    Module.UseLinearForce=false
   )
   /Niagara/Modules/Update/Forces/LineAttractionForce(
    Attraction Strength:NiagaraFloat=&FloatFromCurve(FloatCurve=DataInterface() {propertyValues {Curve {keys [<at 0 1>,<at 0.2 0>]}}}, CurveIndex=$Particles.StartEdge, Scale Curve=1333)
    Line End:Vector3f=&TransformPosition(
     Position=&AddVectorToPosition(Position=Linked() {linkedVariable {name=System.SampleSkeletalMesh.SampledPosition, type=Vector3f}}, Vector=<0 0 -180>)
     Source Space=::World
     Destination Space=::Simulation
    )
    Line Start:Vector3f=&TransformPosition(
     Position=Linked() {linkedVariable {name=System.SampleSkeletalMesh.SampledPosition, type=Vector3f}}
     Source Space=::World
     Destination Space=::Simulation
    )
    Use Attraction Falloff=false
   )
   &LineAttractionForce(
    Attraction Strength:NiagaraFloat=&FloatFromCurve(
     FloatCurve=DataInterface() {propertyValues {Curve {keys [{interpMode=RCIM_Cubic, tangentMode=RCTM_Break, time=0.4, value=0}, <at 0.6 1>]}}}
     CurveIndex=$Particles.StartEdge
     Scale Curve=1000
    )
    Line End:Vector3f=<0 0 0>
    Line Start:Vector3f=&TransformPosition(Position=$System.SampleSkeletalMesh001.SampledPosition, Source Space=::World, Destination Space=::Simulation)
    Use Attraction Falloff=false
   ) {module=LineAttractionForce001}
   /SetVariables(Transient.PhysicsForce:Vector3f=&Multiply_VectorByFloat(Vector=$Transient.PhysicsForce, Float=&Power(a=$Particles.Alpha, b=3))) {isSetParameters=true, module=SetVariables_#1}
   &SolveForcesAndVelocity(
    Acceleration Limit=1
    Clamp Velocity=true
    Limit Acceleration=false
    Speed Limit:NiagaraFloat=&FloatFromCurve(
     FloatCurve=DataInterface() {propertyValues {Curve {keys [{interpMode=RCIM_Cubic, tangentMode=RCTM_Break, time=0, value=0.5}, <at 1.5 1.5>]}}}
     CurveIndex=$Particles.StartEdge
     Scale Curve=200
    )
   )
   &CameraOffset(Camera Offset Amount=20)
   &ScaleMeshSize(Scale Factor=&VectorFromFloat(Value=&FloatFromCurve(
      FloatCurve=DataInterface() {propertyValues {Curve {keys [{interpMode=RCIM_Cubic, leaveTangent=-0.0206, tangentMode=RCTM_Break, time=0, value=0.8}, <at 2 0.5>]}}}
      CurveIndex=$Particles.StartEdge
     )))
   /SetVariables(Particles.ColorC:LinearColor=&MultiplyLinearColor(
     Linear Color=<rgba 0 0.5809 1 1>
     Scale Factor=&MakeVector4(
      X=1
      Y=1
      Z=1
      W=&Multiply_Float(
       A=&FloatFromCurve(FloatCurve=DataInterface() {propertyValues {Curve {keys [<at 0.4 1>,<at 0.5 0>]}}}, Scale Curve=0.03)
       B=&FloatFromCurve(
        FloatCurve=DataInterface() {propertyValues {Curve {keys [{interpMode=RCIM_Cubic, tangentMode=RCTM_Break, time=25, value=0}, {interpMode=RCIM_Cubic, tangentMode=RCTM_Break, time=125, value=1}]}}}
        CurveIndex=&DistanceBetweenPositions(A=$Particles.Position, B=<0 0 0>)
        Scale Curve=$Particles.Alpha
       )
      )
     )
    )) {isSetParameters=true, module=SetVariables_#2}
   &ScaleColor(
    Scale Alpha=&FloatFromCurve(FloatCurve=DataInterface() {propertyValues {Curve {keys [<at 0.4 1>,<at 0.5 0>]}}})
    ScaleA=true
    ScaleRGB=true
   ) {module=ScaleColor001}
   &ScaleColor(
    Scale Alpha=&FloatFromCurve(
     FloatCurve=DataInterface() {propertyValues {Curve {keys [{interpMode=RCIM_Cubic, tangentMode=RCTM_Break, time=25, value=0}, {interpMode=RCIM_Cubic, tangentMode=RCTM_Break, time=125, value=1}]}}}
     CurveIndex=&DistanceBetweenPositions(A=$Particles.Position, B=<0 0 0>)
    )
    ScaleA=true
    ScaleRGB=true
   ) {module=ScaleColor002}
  ]
  rendererClasses [&NiagaraMeshRendererProperties, &NiagaraSpriteRendererProperties]
  renderers [
   {rendererClass=&NiagaraMeshRendererProperties, rendererIndex=0}
   {rendererClass=&NiagaraMeshRendererProperties, rendererIndex=1}
   {rendererClass=&NiagaraSpriteRendererProperties, rendererIndex=2}
  ]
 }
 Empty002 {
  emitterUpdateScript [
   &EmitterState(Life Cycle Mode=::NewEnumerator1, Loop Behavior=::NewEnumerator1, Loop Duration=2, UseLoopDelay=false)
   &SpawnBurst_Instantaneous(Spawn Count=14, Use Spawn Probability=false, UseLoopCountLimit=false)
  ]
  particleSpawnScript [
   &FN_SetSpawnGroup(Inclusivity=true, Normalize Spawn Group=true, Number of Spawn Groups=14, Particles Per Group=1)
   &InitializeParticle(
    Lifetime=4
    Mesh Scale Mode=::NewEnumerator1
    Mesh Uniform Scale Max=1.5
    Mesh Uniform Scale Min=0.8
    Sprite Size Mode=::NewEnumerator3
    Uniform Sprite Size=13.0186
    UsePositionOffset=false
    Write Material Random=true
   )
   &self:CustomBoneSample(
    NormalizedSpawnGroup:NiagaraFloat=$Particles.Normalized Spawn Group
    Skeletal Mesh:NiagaraDataInterfaceSkeletalMesh=$User.TargetPawn
    SpawnGroup:NiagaraInt32=$Particles.SpawnGroup
   ) {module=CustomBoneSample}
   &ShapeLocation(
    MaskSpawnGroup=false
    Override Seed=false
    Shape Origin=$StackContext.SampledPosition
    Sphere Radius=&SelectFloatFromArray(
     Direct Array Index=$Particles.SpawnGroup
     Float Selection Array=DataInterface() {propertyValues {FloatData [10,10,10,7,7,5,5,18,10,13,13,8,8,18]}}
    )
    UseOwnerScale=true
   )
  ]
  particleUpdateScript [
   &ParticleState()
   &self:AlphaDependOnHeight(
    DivideZ=80
    HeadPos:Vector3f=$System.SampleSkeletalMesh.SampledPosition
    ZOffset:NiagaraFloat=&Lerp_Float(
     A=250
     B=0
     Alpha=&ClampFloat(Float=&Subtract_Float(A=1, B=&Multiply_Float(A=$Emitter.Age, B=1)))
    )
   ) {module=AlphaDependOnHeight}
   &self:AlphaDependOnHeight(
    DivideZ=70
    ZOffset:NiagaraFloat=&Lerp_Float(
     A=2665.4104
     B=7.5492
     Alpha=&ClampFloat(Float=&Subtract_Float(A=1, B=&Multiply_Float(A=$Emitter.Age, B=1)))
    )
   ) {enabled=false, module=AlphaDependOnHeight001}
   &UpdateMeshOrientation(
    Rotation Rate=&Lerp_Float(
     A=0
     B=&RandomRangeFloat(Minimum=-3, Maximum=3)
     Alpha=&FloatFromCurve(FloatCurve=DataInterface() {propertyValues {Curve {keys [<at 0 0>,<at 0.1 1>]}}}, CurveIndex=$Particles.StartEdge)
    )
    Rotation Vector=&RandomRangeVector(Minimum=<-1 -1 -1>, Maximum=<1 1 1>)
   )
   &ScaleColor(Scale Alpha=$Particles.Alpha, ScaleA=true, ScaleRGB=true)
   &CurlNoiseForce(
    Noise Frequency=7
    Noise Strength=&FloatFromCurve(
     FloatCurve=DataInterface() {propertyValues {Curve {keys [
         {interpMode=RCIM_Cubic, tangentMode=RCTM_Break, time=0.1, value=1}
         {arriveTangent=0.1541, interpMode=RCIM_Cubic, tangentMode=RCTM_Break, time=1.8, value=0.3}
        ]}}}
     CurveIndex=$Particles.StartEdge
     Scale Curve=16000
    )
    Pan Noise=true
    Pan Noise Field=<0 0 0.1>
   )
   &PointForce(
    Force Strength=&FloatFromCurve(
     FloatCurve=DataInterface() {propertyValues {Curve {keys [{interpMode=RCIM_Cubic, leaveTangent=-1.6096, tangentMode=RCTM_Break, time=0, value=-0.5}, <at 0.5 -1>]}}}
     CurveIndex=$Particles.StartEdge
     Scale Curve=9000
    )
    Module.Use Falloff=false
    Module.UseLinearForce=false
   )
   &LineAttractionForce(
    Attraction Strength:NiagaraFloat=&FloatFromCurve(FloatCurve=DataInterface() {propertyValues {Curve {keys [<at 0 1>,<at 0.2 0>]}}}, CurveIndex=$Particles.StartEdge, Scale Curve=1333)
    Line End:Vector3f=&TransformPosition(
     Position=&AddVectorToPosition(Position=Linked() {linkedVariable {name=System.SampleSkeletalMesh.SampledPosition, type=Vector3f}}, Vector=<0 0 -180>)
     Source Space=::World
     Destination Space=::Simulation
    )
    Line Start:Vector3f=&TransformPosition(
     Position=Linked() {linkedVariable {name=System.SampleSkeletalMesh.SampledPosition, type=Vector3f}}
     Source Space=::World
     Destination Space=::Simulation
    )
    Use Attraction Falloff=false
   )
   &LineAttractionForce(
    Attraction Strength:NiagaraFloat=&FloatFromCurve(
     FloatCurve=DataInterface() {propertyValues {Curve {keys [{interpMode=RCIM_Cubic, tangentMode=RCTM_Break, time=0.4, value=0}, <at 0.6 1>]}}}
     CurveIndex=$Particles.StartEdge
     Scale Curve=1000
    )
    Line End:Vector3f=<0 0 0>
    Line Start:Vector3f=&TransformPosition(Position=$System.SampleSkeletalMesh001.SampledPosition, Source Space=::World, Destination Space=::Simulation)
    Use Attraction Falloff=false
   ) {module=LineAttractionForce001}
   /SetVariables(Transient.PhysicsForce:Vector3f=&Multiply_VectorByFloat(Vector=$Transient.PhysicsForce, Float=&Power(a=$Particles.Alpha, b=3))) {isSetParameters=true, module=SetVariables_#1}
   &SolveForcesAndVelocity(
    Acceleration Limit=1
    Clamp Velocity=true
    Limit Acceleration=false
    Speed Limit:NiagaraFloat=&FloatFromCurve(
     FloatCurve=DataInterface() {propertyValues {Curve {keys [{interpMode=RCIM_Cubic, tangentMode=RCTM_Break, time=0, value=0.5}, <at 1.5 1.5>]}}}
     CurveIndex=$Particles.StartEdge
     Scale Curve=200
    )
   )
   &CameraOffset(Camera Offset Amount=20)
   /SetVariables(Particles.ColorC:LinearColor=&MultiplyLinearColor(
     Linear Color=<rgba 0 0.5809 1 1>
     Scale Factor=&MakeVector4(
      X=1
      Y=1
      Z=1
      W=&Multiply_Float(
       A=&FloatFromCurve(FloatCurve=DataInterface() {propertyValues {Curve {keys [<at 0.4 1>,<at 0.6 0>]}}}, Scale Curve=0.03)
       B=&FloatFromCurve(
        FloatCurve=DataInterface() {propertyValues {Curve {keys [{interpMode=RCIM_Cubic, tangentMode=RCTM_Break, time=25, value=0}, {interpMode=RCIM_Cubic, tangentMode=RCTM_Break, time=125, value=1}]}}}
        CurveIndex=&DistanceBetweenPositions(A=$Particles.Position, B=<0 0 0>)
        Scale Curve=$Particles.Alpha
       )
      )
     )
    )) {isSetParameters=true, module=SetVariables_#2}
   &ScaleColor(
    Scale Alpha=&FloatFromCurve(FloatCurve=DataInterface() {propertyValues {Curve {keys [<at 0.4 1>,<at 0.5 0>]}}})
    ScaleA=true
    ScaleRGB=true
   ) {module=ScaleColor001}
   &ScaleColor(
    Scale Alpha=&FloatFromCurve(
     FloatCurve=DataInterface() {propertyValues {Curve {keys [{interpMode=RCIM_Cubic, tangentMode=RCTM_Break, time=25, value=0}, {interpMode=RCIM_Cubic, tangentMode=RCTM_Break, time=125, value=1}]}}}
     CurveIndex=&DistanceBetweenPositions(A=$Particles.Position, B=<0 0 0>)
    )
    ScaleA=true
    ScaleRGB=true
   ) {module=ScaleColor002}
  ]
  rendererClasses [&NiagaraMeshRendererProperties, &NiagaraSpriteRendererProperties]
  renderers [
   {rendererClass=&NiagaraMeshRendererProperties, rendererIndex=0}
   {rendererClass=&NiagaraMeshRendererProperties, rendererIndex=1}
   {rendererClass=&NiagaraSpriteRendererProperties, rendererIndex=2}
  ]
 }
 Empty003 {
  emitterUpdateScript [
   &EmitterState(Life Cycle Mode=::NewEnumerator1, Loop Behavior=::NewEnumerator1, Loop Duration=3, UseLoopDelay=false)
   &SpawnParticlesFromOtherEmitter(
    Attribute Reader=DataInterface() {propertyValues {EmitterBinding {bindingMode=Other, emitterName=Empty002}}}
    Spawn Rate=15
    Spawn Rate Per Particle Cap=1000
    Use Maximum Age=false
    Use Minimum Age=false
   )
  ]
  particleSpawnScript [
   &InitializeParticle(
    Lifetime=0.7
    Ribbon Width=30
    Ribbon Width Mode=::NewEnumerator1
    Sprite Size Mode=::NewEnumerator3
    UsePositionOffset=false
    Write Material Random=true
   )
   &SampleParticlesFromOtherEmitter(Particle ID Sampling=::NewEnumerator1, Ribbon UV Distance:ENiagara_DirectReadSamplingMode=::NewEnumerator0)
   /SetVariables(Particles.Color=&MakeLinearColorFromVectorAndFloat("Vector (RGB)"=<1 1 1>, "Float (Alpha)"=0.03)) {isSetParameters=true}
  ]
  particleUpdateScript [
   &ParticleState()
   &SolveForcesAndVelocity(Clamp Velocity=false, Limit Acceleration=false)
   &ScaleRibbonWidth(Ribbon Width Scale=&FloatFromCurve(FloatCurve=DataInterface() {propertyValues {Curve {keys [<at 0 1>,<at 1 0>]}}}))
   &ScaleColor(
    Scale Alpha=&FloatFromCurve(FloatCurve=DataInterface() {propertyValues {Curve {keys [
         {interpMode=RCIM_Cubic, tangentMode=RCTM_User, time=0, value=0}
         {arriveTangent=-2.2705, interpMode=RCIM_Cubic, leaveTangent=-2.2705, tangentMode=RCTM_User, time=0.2, value=1}
         {tangentMode=RCTM_User, time=1, value=0}
        ]}}})
    Scale RGB=&MakeVectorFromLinearColorRGB(Color=&ColorFromCurve(ColorCurve=DataInterface() {propertyValues {
        AlphaCurve {keys [<at 0 1 RCIM_Cubic>,<at 1 1 RCIM_Cubic>]}
        BlueCurve {keys [<at 0 1 RCIM_Cubic>,<at 0.6386 0.7 RCIM_Cubic>]}
        GreenCurve {keys [<at 0 0.5 RCIM_Cubic>,<at 0.6386 0.2333 RCIM_Cubic>]}
        RedCurve {keys [<at 0 0 RCIM_Cubic>,<at 0.6386 0 RCIM_Cubic>]}
       }}))
    ScaleA=true
    ScaleRGB=true
   )
   &ScaleColor(
    Scale Alpha=&FloatFromCurve(
     FloatCurve=DataInterface() {propertyValues {Curve {keys [
         {interpMode=RCIM_Cubic, tangentMode=RCTM_Break, time=0, value=1}
         {arriveTangent=-0.5, interpMode=RCIM_Cubic, leaveTangent=-0.5, time=1, value=0.8}
         <at 2 0>
        ]}}}
     CurveIndex=$Emitter.Age
    )
    ScaleA=true
    ScaleRGB=true
   ) {module=ScaleColor001}
  ]
  rendererClasses [&NiagaraRibbonRendererProperties]
  renderers [{rendererClass=&NiagaraRibbonRendererProperties}]
 }
}
summary {
 systemName=NS_TransisitonIAS_Derez
 emitters [
  {emitterName=Empty, rendererClasses [&NiagaraMeshRendererProperties, &NiagaraSpriteRendererProperties]}
  {emitterName=Empty002, rendererClasses [&NiagaraMeshRendererProperties, &NiagaraSpriteRendererProperties]}
  {emitterName=Empty003, rendererClasses [&NiagaraRibbonRendererProperties]}
 ]
 userVariables [User.TargetPawn {
   type=NiagaraDataInterfaceSkeletalMesh
   defaultValue {
    struct=NiagaraExt_VariableValue_DataInterface
    value {dataInterface=&self:NiagaraDataInterfaceSkeletalMesh_0}
   }
  }]
}
